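# Material function that overlays an animated, audio-reactive equalizer on the Cactus Rocker female body's BaseColor and Emissive, gated by an Alpha input.
# UV1 and front/back masks build a quantized cell grid, lit where parabolic bands, offset by panning noise, pass a threshold.
# Band heights take the maximum of looping bass/treble pulses and sampled or fake audio data, computed per vertex.
# Cells get a two-color gradient, with emissive saturation boosted through negative desaturation.
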MaterialFunction /BRCosmetics/Characters/Player/Female/Medium/Bodies/F_MED_Cactus_Rocker_Alt/Material/MF_Cactus_Rocker_Female_Body_Equalizer
graph {nodes [
  n1 = NamedRerouteUsage()
  n2 = NamedRerouteUsage()
  n3 = NamedRerouteUsage()
  n4 = ScalarParameter() {defaultValue=3, group=Cactus Rocker Equalizer, parameterName=Minimum_Emissive_Strength}
  n5 = Desaturation(None=Multiply(A=NamedRerouteUsage(), B=Max(A=@n3, B=@n4)), Fraction=Constant() {r=-1})
  FunctionOutput(None=BlendMaterialAttributes(
    A=@n1
    B=SetMaterialAttributes!(MaterialAttributes=@n1, BaseColor=@n2, EmissiveColor=@n5) {attributeSetTypes [#1, #2]}
    Alpha=NamedRerouteUsage()
   ))
  n6 = FunctionInput() {bUsePreviewValueAsDefault=true, inputName=Alpha, inputType=FunctionInput_Scalar, sortPriority=1, previewValue=<0 0 1 1>}
  NamedRerouteDeclaration(None=Ceil(None=@n6)) {name=Alpha_In, nodeColor=<rgba 1 0.5 0 1>}
  n7 = NamedRerouteUsage()
  n8 = Ceil(None=Saturate(None=Subtract(A=@n7) {constB=0.6}))
  NamedRerouteDeclaration(None=LinearInterpolate(A=NamedRerouteUsage(), B=NamedRerouteUsage(), Alpha=NamedRerouteDeclaration(None=@n8) {name=Front_Mask, nodeColor=<rgba 1 0.8438 0 1>})) {name=Front_Back_V_Mask, nodeColor=<rgba 0.4531 0 1 1>}
  n9 = MaterialFunctionCall(Float2=FunctionInput() {inputName=Color_Mask, inputType=FunctionInput_Vector2, sortPriority=5}) {materialFunction=&BreakOutFloat2Components}
  NamedRerouteDeclaration(None=@n9.G) {name=ControlMask_Back, nodeColor=<rgba 0.9453 0 1 1>}
  n10 = FunctionInput(Preview=VertexInterpolator(VS=PreSkinnedPosition()).PS) {inputName=Preskinned_Pos, sortPriority=10}
  n11 = ComponentMask(None=@n10) {g=true, r=true}
  n12 = Abs(None=DotProduct(A=Normalize(VectorInput=@n11), B=Constant2Vector() {g=1}))
  NamedRerouteDeclaration(None=Saturate(None=Power(Base=@n12) {constExponent=0.15})) {name=SideMask, nodeColor=<rgba 1 0.5156 0 1>}
  NamedRerouteDeclaration(None=@n6) {name=ID, nodeColor=<rgba 0.1484 0 1 1>}
  n13 = VectorParameter() {parameterName=FakeAudioData, channelNames {a="", b="", g=Amplitude, r=Spectrum}}
  n14 = MaterialFunctionCall!(useSmoothStep=StaticBool(), uvx=Constant() {r=0.5}) {materialFunction=/Game/Accessories/FORT_Backpacks/Backpack_F_MED_Cactus_Rocker/Materials/SampleAudioMPCHistory}
  n15 = MaterialFunctionCall(Float4=@n14.Interpolated Spectrum) {materialFunction=&BreakOutFloat4Components}
  n16 = Add(A=@n15.R, B=@n15.R)
  n17 = Divide(A=Add(A=Add(A=@n16, B=@n15.B), B=@n15.A)) {constB=4}
  n18 = StaticSwitchParameter(True=Add(A=@n13.Spectrum, B=@n13.Spectrum), False=Add(A=@n17, B=@n14.Interpolated Amplitude Music)) {group=SkinFX, parameterName=UseFakeAudioData}
  n19 = ScalarParameter() {defaultValue=1, group=Cactus Rocker Equalizer, parameterName=Event Audio Multiplier, sliderMax=1}
  NamedRerouteDeclaration(None=Min(A=Multiply(A=@n18, B=@n19)) {constB=30, desc=prevent crazy values}) {name=AudioSpectrum, nodeColor=<rgba 1 0.8203 0 1>}
  n20 = TextureCoordinate() {coordinateIndex=1}
  n21 = ScalarParameter() {defaultValue=128, group=Cactus Rocker Equalizer, parameterName=Levels Number}
  n22 = AppendVector(A=ScalarParameter() {defaultValue=48.5, group=Cactus Rocker Equalizer, parameterName=Columns Number}, B=@n21)
  n23 = Multiply(
   A=MaterialFunctionCall(X=ComponentMask(None=@n20) {r=true}, Y=NamedRerouteUsage()) {materialFunction=&MakeFloat2}.Result
   B=@n22
  )
  n24 = Frac(None=@n23)
  n25 = Abs(None=Add(A=Multiply(A=@n24) {constB=2}) {constB=-1})
  NamedRerouteDeclaration(None=SphereMask(
    A=ComponentMask(None=@n25) {r=true}
    B=Constant()
    Radius=ScalarParameter() {defaultValue=0.6, group=Cactus Rocker Equalizer, parameterName=Cell Width, sliderMax=1}
   ) {attenuationRadius=0.4}) {name=Cell Mask, nodeColor=<rgba 1 0 0.3281 1>}
  n26 = NamedRerouteUsage()
  n27 = ScalarParameter() {defaultValue=6, parameterName=Color_Transition_Exp}
  n28 = LinearInterpolate(
   A=VectorParameter() {group=Cactus Rocker Equalizer, parameterName=Top Color, defaultValue=<rgba 0 1 1 1>}.RGB
   B=VectorParameter() {group=Cactus Rocker Equalizer, parameterName=Bottom Color, defaultValue=<rgba 0 0 1 1>}.RGB
   Alpha=Saturate(None=Power(Base=@n26, Exp=@n27))
  )
  n29 = Constant() {r=0.1}
  n30 = Divide(A=Floor(None=@n23), B=@n22)
  n31 = MaterialFunctionCall(Float2=@n30) {materialFunction=&BreakOutFloat2Components}
  n32 = Frac(None=Multiply(A=ComponentMask(None=@n31.R) {r=true}, B=ScalarParameter() {defaultValue=2, group=Cactus Rocker Equalizer, parameterName=WaveMult}) {constB=2})
  n33 = MaterialFunctionCall(
   X=Abs(None=Add(A=@n32) {constB=-0.5})
   Y=Power(Base=@n31.G, Exp=ScalarParameter() {defaultValue=2, group=Cactus Rocker Equalizer, parameterName=WaveExp})
  ) {materialFunction=&MakeFloat2}
  n34 = ComponentMask(None=@n33.Result) {g=true}
  n35 = ScalarParameter() {defaultValue=0.8, group=Cactus Rocker Equalizer, parameterName=EQ_VerticalOffset}
  n36 = ComponentMask(None=@n33.Result) {r=true}
  n37 = ScalarParameter() {defaultValue=3, group=Cactus Rocker Equalizer, parameterName=EQ_BandsWidth}
  n38 = Multiply(A=@n36, B=@n37) {desc=waveA}
  n39 = Multiply(A=@n38, B=@n38)
  n40 = ScalarParameter() {defaultValue=2, group=Cactus Rocker Equalizer, parameterName=EQ_SpeedBass}
  n41 = ScalarParameter() {defaultValue=4, group=Cactus Rocker Equalizer, parameterName=EQ_SpeedTreble}
  n42 = Frac(None=Multiply(A=MaterialFunctionCall(X=@n40, Y=@n41) {materialFunction=&MakeFloat2}.Result, B=Time()))
  n43 = Subtract(A=Multiply(A=@n42) {constB=2}) {constB=2}
  n44 = Multiply(A=Multiply(A=@n43, B=@n43), B=@n43)
  n45 = Multiply(
   A=MaterialFunctionCall(
    X=ScalarParameter() {defaultValue=0.6, group=Cactus Rocker Equalizer, parameterName=EQ_MaxIntensityBass}
    Y=ScalarParameter() {defaultValue=0.2, group=Cactus Rocker Equalizer, parameterName=EQ_MaxIntensityTreble}
   ) {materialFunction=&MakeFloat2}.Result
   B=Multiply(A=Multiply(A=@n42, B=@n44), B=Constant() {r=-1.181})
  )
  n46 = ComponentMask(None=@n13) {g=true, r=true}
  n47 = MaterialFunctionCall(X=@n17, Y=@n14.Interpolated Amplitude Music) {materialFunction=&MakeFloat2}
  n48 = Min(A=Power(Base=StaticSwitchParameter(True=@n46, False=@n47.Result) {group=SkinFX, parameterName=UseFakeAudioData})) {constB=30, desc=prevent crazy values}
  n49 = MaterialFunctionCall(Float2=VertexInterpolator(VS=Max(A=@n45, B=@n48)).PS) {materialFunction=&BreakOutFloat2Components}
  n50 = LinearInterpolate(A=@n49.R, B=@n49.R, Alpha=@n36)
  n51 = Subtract(A=@n38, B=Multiply(A=@n37) {constB=0.25}) {desc=waveB}
  n52 = Multiply(A=Multiply(A=@n51, B=@n51)) {constB=-1, desc=waveB}
  n53 = Subtract(A=@n38, B=Multiply(A=@n37) {constB=0.5}) {desc=waveC}
  n54 = Add(A=Multiply(A=Multiply(A=@n53, B=@n53)) {constB=-1, desc=waveC}, B=@n50)
  n55 = Subtract(A=@n38, B=Multiply(A=@n37) {constB=0.75}) {desc=waveD}
  n56 = Multiply(A=@n55, B=@n55)
  n57 = Subtract(B=@n37) {desc=waveD}
  n58 = Multiply(A=@n57, B=@n57)
  n59 = Max(A=Add(A=Multiply(A=@n56) {constB=-1, desc=waveD}, B=@n50), B=Add(A=Multiply(A=@n58) {constB=-1, desc=waveD}, B=@n49.G))
  n60 = Max(A=Add(A=Multiply(A=@n39) {constB=-1, desc=waveA}, B=@n50), B=Max(A=Add(A=@n52, B=@n50), B=Max(A=@n54, B=@n59)))
  n61 = ScalarParameter() {defaultValue=0.3, group=Cactus Rocker Equalizer, parameterName=NoiseIntensity, sliderMax=1}
  n62 = VectorParameter() {
   group=Cactus Rocker Equalizer
   parameterName=Noise Speed
   channelNames {a="", b="", g=NoiseSpeedY, r=NoiseSpeedX}
   defaultValue=<rgba -0.3 0.2 0 0>
  }
  n63 = TextureSampleParameter2D(UVs=Panner(Coordinate=NamedRerouteUsage(), Speed=ComponentMask(None=@n62) {g=true, r=true}) {speedX=1, speedY=1}) {
   group=Cactus Rocker Equalizer
   parameterName=NoiseTexture
   samplerType=SAMPLERTYPE_Masks
   texture=/Game/Global/Textures/Default/Noise/T_CellularNoise
  }
  n64 = Add(
   A=Add(A=Subtract(A=@n34, B=@n35), B=Saturate(None=@n60))
   B=LinearInterpolate(A=@n61, B=Multiply(A=@n61) {constB=-1}, Alpha=@n63.R) {constA=-0.5, constB=0.3}
  )
  n65 = Saturate(None=Power(Base=NamedRerouteUsage(), Exp=ScalarParameter() {defaultValue=3, parameterName=EdgeFade_Exp}))
  n66 = Multiply(A=Multiply(A=NamedRerouteUsage(), B=Step(Y=@n29, X=@n64)), B=LinearInterpolate(B=OneMinus(None=@n65), Alpha=NamedRerouteUsage()) {constA=1})
  NamedRerouteDeclaration(None=Multiply(A=@n28, B=@n66)) {name=E_Out, nodeColor=<rgba 1 0.8906 0 1>}
  n67 = ScalarParameter() {defaultValue=2.5, group=Cactus Rocker Equalizer, parameterName=BaseColor_Boost}
  NamedRerouteDeclaration(None=LinearInterpolate(A=NamedRerouteUsage(), B=Multiply(A=@n67, B=@n28), Alpha=@n66)) {name=BC _Out, nodeColor=<rgba 0 1 0.9766 1>}
  n68 = FunctionInput() {inputType=FunctionInput_MaterialAttributes}
  NamedRerouteDeclaration(None=GetMaterialAttributes(None=NamedRerouteDeclaration(None=@n68) {name=MA_In, nodeColor=<rgba 1 0.5 0 1>}) {attributeGetTypes [#1]}.BaseColor) {name=BC_In, nodeColor=<rgba 0 1 0.9766 1>}
  n69 = Constant() {r=1}
  NamedRerouteDeclaration(None=Divide(A=@n30, B=AppendVector(A=@n69, B=@n21))) {name=UVs, nodeColor=<rgba 1 0 0.4687 1>}
  NamedRerouteDeclaration(None=@n9.R) {name=ControlMask_Front, nodeColor=<rgba 0.5937 1 0 1>}
  Panner(Speed=ComponentMask(None=VectorParameter() {group=Cactus Rocker Equalizer, parameterName=PanSpeed, defaultValue=<rgba 0.35 0 0 0>}.RGB) {g=true, r=true}) {speedX=0.2}
  Saturate(None=Sign(None=ComponentMask(None=@n10) {g=true}))
 ]}
