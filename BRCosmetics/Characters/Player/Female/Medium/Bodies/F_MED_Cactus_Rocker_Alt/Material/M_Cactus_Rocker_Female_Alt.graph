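# Character material for the alternate female Cactus Rocker body skin, built for skeletal meshes on material attributes with a per-expression shading model.
# Shared PreFX and CharacterSkinFX functions feed a body equalizer function, then character PostFX and cinematic PostFX.
# The equalizer takes one FX mask texture sampled twice: UV0 red as alpha, UV1 green and blue as color mask, plus pre-skinned local position.
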Material /BRCosmetics/Characters/Player/Female/Medium/Bodies/F_MED_Cactus_Rocker_Alt/Material/M_Cactus_Rocker_Female_Alt
graph {
 properties {
  bAutomaticallySetUsageInEditor=false
  bUseMaterialAttributes=true
  bUsedWithMeshDeformer=true
  bUsedWithMorphTargets=true
  bUsedWithSkeletalMesh=true
  pixelDepthOffsetMode=PDOM_Legacy
  refractionMethod=RM_IndexOfRefraction
  shadingModel=MSM_FromMaterialExpression
  subsurfaceProfile=&SS_HeroSkin_02
  usedShadingModels=Default Lit | Subsurface | Clear Coat | Subsurface Profile | Cloth | Eye
 }
 nodes [
  n1 = MaterialFunctionCall() {materialFunction=&BaseFN_CharacterFunction_PreFX}
  n2 = TextureSample(UVs=TextureCoordinate()) {
   samplerSource=SSM_Wrap_WorldGroupSettings
   samplerType=SAMPLERTYPE_Masks
   texture=/BRCosmetics/Characters/Player/Female/Medium/Bodies/F_MED_Cactus_Rocker_Alt/Textures/T_F_MED_Cactus_Rocker_Alt_Body_FX_Extra
  }
  n3 = TextureSample(UVs=TextureCoordinate() {coordinateIndex=1}) {
   samplerSource=SSM_Wrap_WorldGroupSettings
   samplerType=SAMPLERTYPE_Masks
   texture=&T_F_MED_Cactus_Rocker_Alt_Body_FX_Extra
  }
  n4 = MaterialFunctionCall!(
   In=MaterialFunctionCall(
    PreskinnedPosition=@n1.Pre skinned Local Position
    Gmap=@n1.Pre skinned Local Position
    MaterialAttributes=@n1.Pre skinned Local Position
    AO=@n1.Pre skinned Local Position
   ) {materialFunction=&CharacterSkinFX}.Result
   Alpha=@n2.R
   Color_Mask=ComponentMask(None=@n3.RGB) {b=true, g=true}
   Preskinned_Pos=@n1.Pre skinned Local Position
  ) {materialFunction=&/MF_Cactus_Rocker_Female_Body_Equalizer}
  n5 = MaterialFunctionCall(AO=@n1.AO, Distance Blend=@n1.AO, Fresnel=@n1.AO, Gmap=@n1.AO, Pre Skinned Local Position=@n1.AO, Material=@n4.Result) {materialFunction=&BaseFN_CharacterFunction_PostFX}
  MaterialFunctionCall(In=@n5.Result) {materialFunction=&CINE_MF_CharPostFX}
 ]
}
